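# Niagara system for the front-end pedestal that emits small, tumbling mesh cubes rising from a ring.
# Source003, the enabled cube emitter, moves particles with curl noise, drag and curve-ramped gravity, pulsing their scale over life.
# A ring-direction dot-product test can expire particles by setting Age to Lifetime, and a scratch module receives the second curl-noise output.
# Three disabled Source variants gate spawning by RankFilter; Glitch_Lines and Rays list no renderers.
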NiagaraSystem /Game/Environments/FrontEnd/FX/Pedestal/NS_Pedestal_Cubes
dependencies {
 usedDataInterfaces [&NiagaraDataInterfaceCurve, &NiagaraDataInterfaceArrayColor]
 usedDynamicInputs [&SelectLinearColorFromArray]
 usedModules [
  &self:SystemSpawnScript.SystemScriptSource.SystemScriptGraph.NiagaraNodeAssignment_0.SetVariables_#1
  &self:SystemSpawnScript.SystemScriptSource.SystemScriptGraph.NiagaraNodeAssignment_1.SetVariables_#2
  /Game/Effects/Niagara/Modules/Water/HSV_Offset
  &SystemState
 ]
}
emitters {
 Glitch_Lines {rendererClasses [], renderers []}
 Rays {rendererClasses [], renderers []}
 Source {
  bEnabled=false
  emitterUpdateScript [
   &EmitterState()
   &SpawnRate(
    SpawnRate=&MaskFloat(Mask Source=1, Mask Parameter=$System.RankFilter, Mask=0, Masked Float=25, Passthrough Float=0)
    Use Spawn Probability=false
   )
  ]
  particleSpawnScript [
   &InitializeParticle(
    Color=<rgba 0 0.5809 1 1>
    Lifetime=&RandomRangeFloat(Minimum=1.75, Maximum=2.5)
    Mesh Scale Mode=::NewEnumerator1
    Mesh Uniform Scale Min=1.15
    Sprite Size Mode=::NewEnumerator3
    Uniform Sprite Size=13
    UsePositionOffset=false
    Write Material Random=true
   )
   &ShapeLocation(
    MaskSpawnGroup=false
    Override Seed=false
    Ring / Disc Distribution Mode:ENiagaraSphereDistributionMode=::NewEnumerator1
    Ring Radius=40
    Shape Primitive=::NewEnumerator4
    U Position=&RandomRangeFloat(Minimum=0, Maximum=1)
    UseOwnerScale=true
   )
   &AddVelocity(Velocity=<0 0 25>, Velocity Speed Scale=&RandomRangeFloat(Minimum=1, Maximum=1.5))
  ]
  particleUpdateScript [
   &ParticleState()
   &UpdateMeshOrientation(
    Rotation Rate=&Lerp_Float(
     A=0
     B=&RandomRangeFloat(Minimum=-3, Maximum=3)
     Alpha=&FloatFromCurve(FloatCurve=DataInterface() {propertyValues {Curve {keys [<at 0 0>,<at 0.1 1>]}}}, CurveIndex=$Particles.StartEdge)
    )
    Rotation Vector=&RandomRangeVector(Minimum=<-1 -1 -1>, Maximum=<1 1 1>)
   )
   &ScaleColor(ScaleA=true, ScaleRGB=true)
   &CurlNoiseForce(
    Noise Frequency=22
    Noise Strength=&FloatFromCurve(
     FloatCurve=DataInterface() {propertyValues {Curve {
        defaultValue=1
        keys [{interpMode=RCIM_Cubic, tangentMode=RCTM_User, time=0.15, value=0.2}, {interpMode=RCIM_Cubic, tangentMode=RCTM_User, time=0.5424, value=1}]
       }}}
     Scale Curve=80
    )
    Pan Noise=true
    Pan Noise Field=<0 0 0.2>
   )
   &CurlNoiseForce(
    Bias Noise=false
    Noise Frequency=8
    Noise Strength=1
    Pan Noise=true
    Randomize Noise Sample=false
    Sample Position:NiagaraPosition=$Particles.Initial.Position
    Scale Curl Lookup By Engine Owner Scale=false
    Write to Intrinsic Parameters=false
   ) {module=CurlNoiseForce001}
   &Drag(Drag=0.4, Use Linear Drag=true, Use Rotational Drag=true)
   &GravityForce(Gravity=&Multiply_VectorByFloat(
     Vector=<0 0 -100>
     Float=&FloatFromCurve(
      FloatCurve=DataInterface() {propertyValues {Curve {
         defaultValue=1
         keys [{interpMode=RCIM_Cubic, tangentMode=RCTM_User, time=0.2983, value=0}, {interpMode=RCIM_Cubic, tangentMode=RCTM_User, time=1, value=1}]
        }}}
      Scale Curve=&RandomRangeFloat(Minimum=0, Maximum=1)
     )
    ))
   &SolveForcesAndVelocity(Acceleration Limit=1, Clamp Velocity=false, Limit Acceleration=false)
   &CameraOffset(Camera Offset Amount=-25)
   &ScaleMeshSize(Scale Factor=&VectorFromFloat(Value=&FloatFromCurve(FloatCurve=DataInterface() {propertyValues {Curve {
         defaultValue=0.8
         keys [
          {interpMode=RCIM_Cubic, tangentMode=RCTM_User, time=0, value=0}
          {arriveTangent=-0.6724, interpMode=RCIM_Cubic, leaveTangent=-0.6724, tangentMode=RCTM_User, time=0.1298, value=1.4086}
          {arriveTangent=-1.7567, interpMode=RCIM_Cubic, leaveTangent=-1.7567, time=0.4207, value=1.2186}
          {arriveTangent=-4.0472, interpMode=RCIM_Cubic, leaveTangent=-4.0472, tangentMode=RCTM_User, time=0.681, value=0.4403}
          {tangentMode=RCTM_User, time=1, value=0}
         ]
        }}})))
   /SetVariables(Particles.Age:NiagaraFloat=&MakeCustomFloatFromBool(
     Bool=&SetBoolByFloatComparison(
      A=&DotProduct(
       Vector A=&NormalizeVector(Vector To Normalize=$Particles.ShapeLocation.ShapeVector, Safe Normalize=true, Safe Fallback Threshold=0, Safe Fallback Vector=<0 0 0.001>)
       Vector B=<0 1 0>
      )
      B=1
      Comparison Type=::NewEnumerator6
     )
     True Float=$Particles.Age
     False Float=$Particles.Lifetime
    )) {isSetParameters=true, module=SetVariables_#3}
   &self:ScratchModule(CurlNoise Direction:Vector3f=$Output.CurlNoiseForce001.CurlNoiseForce) {module=ScratchModule}
  ]
  rendererClasses [&NiagaraMeshRendererProperties, &NiagaraSpriteRendererProperties]
  renderers [
   {rendererClass=&NiagaraMeshRendererProperties, rendererIndex=0}
   {rendererClass=&NiagaraSpriteRendererProperties, rendererIndex=1}
   {rendererClass=&NiagaraMeshRendererProperties, rendererIndex=2}
   {rendererClass=&NiagaraMeshRendererProperties, rendererIndex=3}
  ]
 }
 Source001 {
  bEnabled=false
  emitterUpdateScript [
   &EmitterState()
   &SpawnRate(
    SpawnRate=&MaskFloat(Mask Source=1, Mask Parameter=$System.RankFilter, Mask=1, Masked Float=25, Passthrough Float=0)
    Use Spawn Probability=false
   )
  ]
  particleSpawnScript [
   &InitializeParticle(
    Color=<rgba 0 0.5809 1 1>
    Lifetime=&RandomRangeFloat(Minimum=1.5, Maximum=2.6)
    Mesh Scale Mode=::NewEnumerator1
    Mesh Uniform Scale Min=1.15
    Sprite Size Mode=::NewEnumerator3
    Uniform Sprite Size=13
    UsePositionOffset=false
    Write Material Random=true
   )
   &ShapeLocation(
    MaskSpawnGroup=false
    Override Seed=false
    Ring / Disc Distribution Mode:ENiagaraSphereDistributionMode=::NewEnumerator1
    Ring Radius=40
    Shape Primitive=::NewEnumerator4
    U Position=&RandomRangeFloat(Minimum=0, Maximum=1)
    UseOwnerScale=true
   )
   &AddVelocity(Velocity=<0 0 20>, Velocity Speed Scale=&RandomRangeFloat(Minimum=1, Maximum=2))
  ]
  particleUpdateScript [
   &ParticleState()
   &UpdateMeshOrientation(
    Rotation Rate=&Lerp_Float(
     A=0
     B=&RandomRangeFloat(Minimum=-3, Maximum=3)
     Alpha=&FloatFromCurve(FloatCurve=DataInterface() {propertyValues {Curve {keys [<at 0 0>,<at 0.1 1>]}}}, CurveIndex=$Particles.StartEdge)
    )
    Rotation Vector=&RandomRangeVector(Minimum=<-1 -1 -1>, Maximum=<1 1 1>)
   )
   &ScaleColor(ScaleA=true, ScaleRGB=true)
   &CurlNoiseForce(
    Noise Frequency=22
    Noise Strength=&FloatFromCurve(
     FloatCurve=DataInterface() {propertyValues {Curve {
        defaultValue=1
        keys [{interpMode=RCIM_Cubic, tangentMode=RCTM_User, time=0.15, value=0.2}, {interpMode=RCIM_Cubic, tangentMode=RCTM_User, time=0.5424, value=1}]
       }}}
     Scale Curve=88
    )
    Pan Noise=true
    Pan Noise Field=<0 0 0.2>
   ) {enabled=false}
   &CurlNoiseForce(
    Bias Noise=false
    Noise Frequency=33
    Noise Strength=1
    Pan Noise=true
    Randomize Noise Sample=false
    Sample Position:NiagaraPosition=$Particles.Initial.Position
    Scale Curl Lookup By Engine Owner Scale=false
    Write to Intrinsic Parameters=false
   ) {module=CurlNoiseForce001}
   &Drag(Use Linear Drag=true, Use Rotational Drag=true)
   &GravityForce(Gravity=&Multiply_VectorByFloat(
     Vector=<0 0 -80>
     Float=&FloatFromCurve(
      FloatCurve=DataInterface() {propertyValues {Curve {
         defaultValue=1
         keys [{interpMode=RCIM_Cubic, tangentMode=RCTM_User, time=0.2983, value=0}, {interpMode=RCIM_Cubic, tangentMode=RCTM_User, time=1, value=1}]
        }}}
      Scale Curve=&RandomRangeFloat(Minimum=0, Maximum=1)
     )
    )) {enabled=false}
   &MeshRotationForce(Rotation=&Multiply_VectorByFloat(
     Vector=&RandomRangeVector(Minimum=<-1 -1 -1>, Maximum=<1 1 1>)
     Float=&RandomRangeFloat(Minimum=0, Maximum=3)
    ))
   &SolveForcesAndVelocity(Acceleration Limit=1, Clamp Velocity=false, Limit Acceleration=false)
   &CameraOffset(Camera Offset Amount=-25)
   &ScaleMeshSize(Scale Factor=&VectorFromFloat(Value=&FloatFromCurve(FloatCurve=DataInterface() {propertyValues {Curve {
         defaultValue=0.8
         keys [
          {interpMode=RCIM_Cubic, tangentMode=RCTM_User, time=0, value=0}
          {arriveTangent=0.7607, interpMode=RCIM_Cubic, leaveTangent=0.7607, time=0.0848, value=1.3956}
          {arriveTangent=-2.1816, interpMode=RCIM_Cubic, leaveTangent=-2.1816, tangentMode=RCTM_User, time=0.753, value=0.5729}
          {tangentMode=RCTM_User, time=1, value=0}
         ]
        }}})))
   /SetVariables(Particles.Age:NiagaraFloat=&MakeCustomFloatFromBool(
     Bool=&SetBoolByFloatComparison(
      A=&DotProduct(
       Vector A=&NormalizeVector(Vector To Normalize=$Particles.ShapeLocation.ShapeVector, Safe Normalize=true, Safe Fallback Threshold=0, Safe Fallback Vector=<0 0 0.001>)
       Vector B=<0 1 0>
      )
      B=1
      Comparison Type=::NewEnumerator6
     )
     True Float=$Particles.Age
     False Float=$Particles.Lifetime
    )) {isSetParameters=true, module=SetVariables_#3}
   &self:ScratchModule(CurlNoise Direction:Vector3f=$Output.CurlNoiseForce001.CurlNoiseForce) {module=ScratchModule}
  ]
  rendererClasses [&NiagaraMeshRendererProperties, &NiagaraSpriteRendererProperties]
  renderers [
   {rendererClass=&NiagaraMeshRendererProperties, rendererIndex=0}
   {rendererClass=&NiagaraSpriteRendererProperties, rendererIndex=1}
   {rendererClass=&NiagaraMeshRendererProperties, rendererIndex=2}
  ]
 }
 Source002 {
  bEnabled=false
  emitterUpdateScript [
   &EmitterState()
   &SpawnRate(
    SpawnRate=&MaskFloat(Mask Source=1, Mask Parameter=$System.RankFilter, Mask=2, Masked Float=2, Passthrough Float=0)
    Use Spawn Probability=false
   )
  ]
  particleSpawnScript [
   &InitializeParticle(
    Color=<rgba 0 0.5809 1 1>
    Lifetime=&RandomRangeFloat(Minimum=3, Maximum=5)
    Mesh Scale Mode=::NewEnumerator1
    Mesh Uniform Scale Min=1.15
    Sprite Size Mode=::NewEnumerator3
    Uniform Sprite Size=13
    UsePositionOffset=false
    Write Material Random=true
   )
   &ShapeLocation(
    MaskSpawnGroup=false
    Override Seed=false
    Ring / Disc Distribution Mode:ENiagaraSphereDistributionMode=::NewEnumerator1
    Ring Radius=40
    Shape Primitive=::NewEnumerator4
    U Position=&RandomRangeFloat(Minimum=0, Maximum=1)
    UseOwnerScale=true
   )
   &AddVelocity(Velocity=<0 0 8>, Velocity Speed Scale=&RandomRangeFloat(Minimum=0.5, Maximum=1))
  ]
  particleUpdateScript [
   &ParticleState()
   &UpdateMeshOrientation(
    Rotation Rate=&Lerp_Float(
     A=0
     B=&RandomRangeFloat(Minimum=-3, Maximum=3)
     Alpha=&FloatFromCurve(FloatCurve=DataInterface() {propertyValues {Curve {keys [<at 0 0>,<at 0.1 1>]}}}, CurveIndex=$Particles.StartEdge)
    )
    Rotation Vector=&RandomRangeVector(Minimum=<-1 -1 -1>, Maximum=<1 1 1>)
   )
   &ScaleColor(ScaleA=true, ScaleRGB=true)
   &CurlNoiseForce(
    Noise Frequency=22
    Noise Strength=&FloatFromCurve(
     FloatCurve=DataInterface() {propertyValues {Curve {
        defaultValue=1
        keys [{interpMode=RCIM_Cubic, tangentMode=RCTM_User, time=0.15, value=0.2}, {interpMode=RCIM_Cubic, tangentMode=RCTM_User, time=0.5424, value=1}]
       }}}
     Scale Curve=88
    )
    Pan Noise=true
    Pan Noise Field=<0 0 0.2>
   ) {enabled=false}
   &CurlNoiseForce(
    Bias Noise=false
    Noise Frequency=33
    Noise Strength=1
    Pan Noise=true
    Randomize Noise Sample=false
    Sample Position:NiagaraPosition=$Particles.Initial.Position
    Scale Curl Lookup By Engine Owner Scale=false
    Write to Intrinsic Parameters=false
   ) {module=CurlNoiseForce001}
   &Drag(Drag=0.1, Use Linear Drag=true, Use Rotational Drag=true)
   &GravityForce(Gravity=&Multiply_VectorByFloat(
     Vector=<0 0 -80>
     Float=&FloatFromCurve(
      FloatCurve=DataInterface() {propertyValues {Curve {
         defaultValue=1
         keys [{interpMode=RCIM_Cubic, tangentMode=RCTM_User, time=0.2983, value=0}, {interpMode=RCIM_Cubic, tangentMode=RCTM_User, time=1, value=1}]
        }}}
      Scale Curve=&RandomRangeFloat(Minimum=0, Maximum=1)
     )
    )) {enabled=false}
   &SolveForcesAndVelocity(Acceleration Limit=1, Clamp Velocity=false, Limit Acceleration=false)
   &CameraOffset(Camera Offset Amount=-25)
   &ScaleMeshSize(Scale Factor=&VectorFromFloat(Value=&FloatFromCurve(FloatCurve=DataInterface() {propertyValues {Curve {
         defaultValue=0.8
         keys [
          {interpMode=RCIM_Cubic, tangentMode=RCTM_User, time=0, value=0}
          {arriveTangent=0.7607, interpMode=RCIM_Cubic, leaveTangent=0.7607, time=0.1395, value=0.9806}
          {arriveTangent=-2.1816, interpMode=RCIM_Cubic, leaveTangent=-2.1816, tangentMode=RCTM_User, time=0.753, value=0.5729}
          {tangentMode=RCTM_User, time=1, value=0}
         ]
        }}})))
   /SetVariables(Particles.Age:NiagaraFloat=&MakeCustomFloatFromBool(
     Bool=&SetBoolByFloatComparison(
      A=&DotProduct(
       Vector A=&NormalizeVector(Vector To Normalize=$Particles.ShapeLocation.ShapeVector, Safe Normalize=true, Safe Fallback Threshold=0, Safe Fallback Vector=<0 0 0.001>)
       Vector B=<0 1 0>
      )
      B=1
      Comparison Type=::NewEnumerator6
     )
     True Float=$Particles.Age
     False Float=$Particles.Lifetime
    )) {isSetParameters=true, module=SetVariables_#3}
   &self:ScratchModule(CurlNoise Direction:Vector3f=$Output.CurlNoiseForce001.CurlNoiseForce) {enabled=false, module=ScratchModule}
  ]
  rendererClasses [&NiagaraMeshRendererProperties, &NiagaraSpriteRendererProperties]
  renderers [
   {rendererClass=&NiagaraMeshRendererProperties, rendererIndex=0}
   {rendererClass=&NiagaraSpriteRendererProperties, rendererIndex=1}
   {rendererClass=&NiagaraMeshRendererProperties, rendererIndex=2}
  ]
 }
 Source003 {
  emitterUpdateScript [
   &EmitterState()
   &SpawnRate(
    SpawnRate=&MaskFloat(Mask Source=1, Mask Parameter=$System.RankFilter, Mask=0, Masked Float=15, Passthrough Float=15)
    Use Spawn Probability=false
   )
  ]
  particleSpawnScript [
   &InitializeParticle(
    Color=<rgba 0 0.5809 1 1>
    Lifetime=&RandomRangeFloat(Minimum=1.75, Maximum=2.2)
    Mesh Scale Mode=::NewEnumerator1
    Mesh Uniform Scale Max=1.5
    Mesh Uniform Scale Min=1.15
    Sprite Size Mode=::NewEnumerator3
    Uniform Sprite Size=13
    UsePositionOffset=false
    Write Material Random=true
   )
   &ShapeLocation(
    MaskSpawnGroup=false
    Override Seed=false
    Ring / Disc Distribution Mode:ENiagaraSphereDistributionMode=::NewEnumerator1
    Ring Radius=40
    Shape Primitive=::NewEnumerator4
    U Position=&RandomRangeFloat(Minimum=0, Maximum=1)
    UseOwnerScale=true
   )
   &AddVelocity(Velocity=<0 0 35>, Velocity Speed Scale=&RandomRangeFloat(Minimum=1, Maximum=1.5))
  ]
  particleUpdateScript [
   &ParticleState()
   &UpdateMeshOrientation(
    Rotation Rate=&Lerp_Float(
     A=0
     B=&RandomRangeFloat(Minimum=-3, Maximum=3)
     Alpha=&FloatFromCurve(FloatCurve=DataInterface() {propertyValues {Curve {keys [<at 0 0>,<at 0.1 1>]}}}, CurveIndex=$Particles.StartEdge)
    )
    Rotation Vector=&RandomRangeVector(Minimum=<-1 -1 -1>, Maximum=<1 1 1>)
   )
   &ScaleColor(ScaleA=true, ScaleRGB=true)
   &CurlNoiseForce(
    Noise Frequency=22
    Noise Strength=&FloatFromCurve(
     FloatCurve=DataInterface() {propertyValues {Curve {
        defaultValue=1
        keys [{interpMode=RCIM_Cubic, tangentMode=RCTM_User, time=0.15, value=0.2}, {interpMode=RCIM_Cubic, tangentMode=RCTM_User, time=0.5424, value=1}]
       }}}
     Scale Curve=80
    )
    Pan Noise=true
    Pan Noise Field=<0 0 0.2>
   )
   &CurlNoiseForce(
    Bias Noise=false
    Noise Frequency=8
    Noise Strength=1
    Pan Noise=true
    Randomize Noise Sample=false
    Sample Position:NiagaraPosition=$Particles.Initial.Position
    Scale Curl Lookup By Engine Owner Scale=false
    Write to Intrinsic Parameters=false
   ) {module=CurlNoiseForce001}
   &Drag(Drag=0.4, Use Linear Drag=true, Use Rotational Drag=true)
   &GravityForce(Gravity=&Multiply_VectorByFloat(
     Vector=<0 0 -100>
     Float=&FloatFromCurve(
      FloatCurve=DataInterface() {propertyValues {Curve {
         defaultValue=1
         keys [{interpMode=RCIM_Cubic, tangentMode=RCTM_User, time=0.2983, value=0}, {interpMode=RCIM_Cubic, tangentMode=RCTM_User, time=1, value=1}]
        }}}
      Scale Curve=&RandomRangeFloat(Minimum=0, Maximum=1)
     )
    ))
   &SolveForcesAndVelocity(Acceleration Limit=1, Clamp Velocity=false, Limit Acceleration=false)
   &CameraOffset(Camera Offset Amount=-25)
   &ScaleMeshSize(Scale Factor=&VectorFromFloat(Value=&FloatFromCurve(FloatCurve=DataInterface() {propertyValues {Curve {
         defaultValue=0.8
         keys [
          {interpMode=RCIM_Cubic, tangentMode=RCTM_User, time=0, value=0}
          {arriveTangent=-0.6724, interpMode=RCIM_Cubic, leaveTangent=-0.6724, tangentMode=RCTM_User, time=0.1298, value=1.4086}
          {arriveTangent=-1.7567, interpMode=RCIM_Cubic, leaveTangent=-1.7567, time=0.4207, value=1.2186}
          {arriveTangent=-4.0472, interpMode=RCIM_Cubic, leaveTangent=-4.0472, tangentMode=RCTM_User, time=0.681, value=0.4403}
          {tangentMode=RCTM_User, time=1, value=0}
         ]
        }}})))
   /SetVariables(Particles.Age:NiagaraFloat=&MakeCustomFloatFromBool(
     Bool=&SetBoolByFloatComparison(
      A=&DotProduct(
       Vector A=&NormalizeVector(Vector To Normalize=$Particles.ShapeLocation.ShapeVector, Safe Normalize=true, Safe Fallback Threshold=0, Safe Fallback Vector=<0 0 0.001>)
       Vector B=<0 1 0>
      )
      B=1
      Comparison Type=::NewEnumerator6
     )
     True Float=$Particles.Age
     False Float=$Particles.Lifetime
    )) {isSetParameters=true, module=SetVariables_#3}
   &self:ScratchModule(CurlNoise Direction:Vector3f=$Output.CurlNoiseForce001.CurlNoiseForce) {module=ScratchModule}
  ]
  rendererClasses [&NiagaraMeshRendererProperties]
  renderers [
   {rendererClass=&NiagaraMeshRendererProperties, rendererIndex=0}
   {rendererClass=&NiagaraMeshRendererProperties, rendererIndex=1}
  ]
 }
}
summary {
 systemName=NS_Pedestal_Cubes
 emitters [
  {
   bEnabled=false
   emitterName=Source
   rendererClasses [&NiagaraMeshRendererProperties, &NiagaraSpriteRendererProperties]
  }
  {
   bEnabled=false
   emitterName=Source001
   rendererClasses [&NiagaraMeshRendererProperties, &NiagaraSpriteRendererProperties]
  }
  {
   bEnabled=false
   emitterName=Source002
   rendererClasses [&NiagaraMeshRendererProperties, &NiagaraSpriteRendererProperties]
  }
  {emitterName=Source003, rendererClasses [&NiagaraMeshRendererProperties]}
  {emitterName=Glitch_Lines, rendererClasses []}
  {emitterName=Rays, rendererClasses []}
 ]
 userVariables [User.RankFilter {type=NiagaraInt32}]
}
